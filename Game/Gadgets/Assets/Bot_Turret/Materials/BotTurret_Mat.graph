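# Masked, lit material for the Bot Turret gadget's skeletal mesh, with a digital-noise phase-in effect.
# An Opacity parameter drives a HeightLerp over tiled noise, whose alpha becomes the opacity mask and, inverted, tinted and squared, an emissive glow.
# That glow adds to a cropped, color-multiplied emissive texture.
# Diffuse, normal and packed specular-mask textures supply the surface, with all attributes routed through Fortnite_BaseFunction.
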Material /Game/Gadgets/Assets/Bot_Turret/Materials/BotTurret_Mat
graph {
 properties {
  bUseMaterialAttributes=true
  bUsedWithMeshDeformer=true
  bUsedWithMeshParticles=true
  bUsedWithSkeletalMesh=true
  blendMode=BLEND_Masked
  pixelDepthOffsetMode=PDOM_Legacy
  previewMesh=/Game/Gadgets/Assets/Bot_Turret/Meshes/SK_Bot_Turret
  refractionMethod=RM_IndexOfRefraction
  usedShadingModels=Default Lit
 }
 nodes [
  n1 = TextureSampleParameter2D() {parameterName=SpecularMasks, texture=/Game/Gadgets/Assets/Bot_Turret/Textures/T_TurretBot_S}
  n2 = ScalarParameter() {defaultValue=0.2, parameterName=SpecRoughtnessMin}
  n3 = ScalarParameter() {defaultValue=0.7744, parameterName=SpecRoughtnessMax}
  n4 = Multiply(A=TextureCoordinate() {uTiling=2, vTiling=2}, B=ScalarParameter() {defaultValue=1, parameterName=Res in texture scale})
  n5 = MaterialFunctionCall(
   Transition Phase=ScalarParameter() {defaultValue=1, parameterName=Opacity}
   Height Texture=DotProduct(
    A=TextureSample(UVs=@n4) {samplerType=SAMPLERTYPE_LinearColor, texture=/Game/Global/Textures/Default/Noise/T_Atlas_DigitalNoise_01}.RGB
    B=Constant() {r=0.3333}
   )
  ) {materialFunction=&HeightLerp}
  n6 = Multiply(A=VectorParameter() {parameterName=Phase In Emissive Color, defaultValue=<rgba 1.6862 2.7774 10 1>}.RGB, B=OneMinus(None=@n5.Alpha))
  n7 = Multiply(A=@n6, B=@n6)
  n8 = VectorParameter() {parameterName=Emissive Texture Cropping, defaultValue=<rgba 0 0 1 1>}
  n9 = Multiply(
   A=MaterialFunctionCall!(
    UVs
    TextureIn=TextureObjectParameter() {parameterName=Emissive, texture=/Game/Gadgets/Assets/Bot_Turret/Textures/T_TurretBot_E}
    UpperLeftCorner=ComponentMask(None=@n8.RGB) {g=true, r=true}
    LowerRightCorner=AppendVector(A=@n8.B, B=@n8.B)
    ExportFloat4
   ) {materialFunction=/Engine/Functions/Engine_MaterialFunctions01/Texturing/TextureCropping}.CroppedMasked
   B=VectorParameter() {parameterName=Emissive Multiplier, defaultValue=<rgba 6 1 0.5 1>}.RGB
  )
  MaterialFunctionCall!(
   Material attributes=MakeMaterialAttributes!(
    BaseColor=TextureSampleParameter2D() {parameterName=Diffuse, texture=/Game/Gadgets/Assets/Bot_Turret/Textures/T_TurretBot_D}.RGB
    Metallic=@n1.G
    Specular=@n1.G
    Roughness=LinearInterpolate(A=@n2, B=@n3, Alpha=@n1.B) {constA=0.2, constB=0.7744}
    Anisotropy
    EmissiveColor=Add(A=@n7, B=@n9)
    Opacity
    OpacityMask=@n5.Alpha
    Normal=TextureSampleParameter2D() {parameterName=Normals, samplerType=SAMPLERTYPE_Normal, texture=/Game/Gadgets/Assets/Bot_Turret/Textures/T_TurretBot_N}.RGB
    Tangent
    WorldPositionOffset
    SubsurfaceColor
    ClearCoat
    ClearCoatRoughness
    AmbientOcclusion
    Refraction
    CustomizedUVs_0
    CustomizedUVs_1
    CustomizedUVs_2
    CustomizedUVs_3
    CustomizedUVs_4
    CustomizedUVs_5
    CustomizedUVs_6
    CustomizedUVs_7
    PixelDepthOffset
    ShadingModel
    Displacement
   )
   IsGameplayAsset=StaticBool()
  ) {materialFunction=/Game/Packages/Fortress_SharedMaterials/Base_Material_Functions/Fortnite_BaseFunction}
 ]
}
