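# Unlit material for the Bot Turret projectile ring, colored by friendly or enemy team.
# An inverted Fresnel blends two inner colors for each team, and the particle dynamic parameter TeamColor lerps from the friendly to the enemy palette.
# Vertex color red masks in a much brighter, team-specific HDR highlight over that inner gradient.
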Material /Game/Gadgets/Assets/Bot_Turret/Materials/M_BotTurret_ProjectileRing_01
graph {
 properties {pixelDepthOffsetMode=PDOM_Legacy, shadingModel=MSM_Unlit, usedShadingModels=Unlit}
 nodes [
  n1 = VectorParameter() {parameterName=Inner Color B Friendly, defaultValue=<rgba 1 0.5254 0.087 0>}
  n2 = VectorParameter() {parameterName=Inner Color A Friendly, defaultValue=<rgba 3 1.7097 0.6386 0>}
  n3 = MaterialFunctionCall(
   Brightness=ScalarParameter() {defaultValue=1, parameterName=Fresnel Brightness}
   Exponent=ScalarParameter() {defaultValue=2, parameterName=Fresnel Exponent}
  ) {materialFunction=&MF_Fresnel_FX}
  n4 = VectorParameter() {parameterName=Inner Color B Enemy, defaultValue=<rgba 1 0.1022 0.087 0>}
  n5 = VectorParameter() {parameterName=Inner Color A Enemy, defaultValue=<rgba 3 0.7566 0.6386 0>}
  n6 = DynamicParameter() {defaultValue=<rgba 0 0 0 1>, paramNames [Param1,Param2,Param3,TeamColor]}
  LinearInterpolate(
   A=LinearInterpolate(
    A=LinearInterpolate(A=@n1.RGB, B=@n2.RGB, Alpha=@n3.Fresnel Inverted)
    B=LinearInterpolate(A=@n4.RGB, B=@n5.RGB, Alpha=@n3.Fresnel Inverted)
    Alpha=@n6.TeamColor
   )
   B=LinearInterpolate(
    A=VectorParameter() {parameterName=Highlight Friendly, defaultValue=<rgba 40 28.9733 3.4596 0>}.RGB
    B=VectorParameter() {parameterName=Highlight Enemy, defaultValue=<rgba 40 4.0686 3.4596 0>}.RGB
    Alpha=@n6.TeamColor
   )
   Alpha=VertexColor().R
  )
 ]
}
