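# Translucent, unlit, two-sided material for the bot turret's scanner ray on mesh particles and skeletal meshes.
# Inverted Fresnel, vertex red and a beam texture panned by two summed sines form a sweep mask driving color and opacity.
# A dynamic parameter blends friendly and enemy palettes, though both fade colors are named "Color Fade Friendly".
# Opacity adds vertex, depth, near-camera, optional distance and edge-mask fades; low quality skips Fresnel and depth fade.
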Material /Game/Gadgets/Assets/Bot_Turret/Materials/M_BotTurret_ScannerRay_01
graph {
 properties {
  bUsedWithMeshParticles=true
  bUsedWithNiagaraMeshParticles=true
  bUsedWithSkeletalMesh=true
  blendMode=BLEND_Translucent
  pixelDepthOffsetMode=PDOM_Legacy
  shadingModel=MSM_Unlit
  twoSided=true
  usedShadingModels=Unlit
 }
 nodes [
  n1 = ScalarParameter() {defaultValue=1.3, parameterName=Fresnel Brightness}
  n2 = MaterialFunctionCall(Brightness=@n1, Exponent=ScalarParameter() {defaultValue=8, parameterName=Fresnel Exponent}) {materialFunction=&MF_Fresnel_FX}
  n3 = QualitySwitch(Default=@n2.Fresnel Inverted, Low=@n1, High=@n2.Fresnel Inverted, Medium=@n2.Fresnel Inverted)
  n4 = VertexColor()
  n5 = TextureCoordinate() {uTiling=8}
  n6 = DynamicParameter() {defaultValue=<rgba 0 0 0 0>, paramNames [ScannerRandomPosition, Speed Random, Param3, Team Color Picker]}
  n7 = AppendVector(A=@n6.ScannerRandomPosition, B=Constant())
  n8 = Multiply(A=Time(), B=ScalarParameter() {defaultValue=9, parameterName=Speed A})
  n9 = Sine(None=@n8) {period=1.3}
  n10 = Sine(None=@n8) {period=1.7}
  n11 = TextureSample!(UVs=Panner(Coordinate=Add(A=@n5, B=@n7), Time=Add(A=@n9, B=@n10)) {speedX=2.6}, Tex, Level, Apply View MipBias) {
   constMipValue=0
   mipValueMode=TMVM_MipLevel
   samplerType=SAMPLERTYPE_LinearColor
   texture=/Game/Effects/Fort_Effects/Textures/Tile/T_Raygun_BeamReveal_01
  }
  n12 = ScalarParameter() {defaultValue=1, parameterName=ScannerSweepOpacity, sliderMax=1}
  n13 = Clamp(None=Add(A=Add(A=@n3, B=@n4.R), B=Multiply(A=@n11.R, B=@n12)))
  n14 = VertexColor()
  n15 = LinearInterpolate(
   A=LinearInterpolate(
    A=VectorParameter() {parameterName=Color Inner Friendly, defaultValue=<rgba 1 0.5221 0.0734 1>}.RGB
    B=VectorParameter() {parameterName=Color Outer Friendly, defaultValue=<rgba 10 6.8161 1.3167 1>}.RGB
    Alpha=@n13
   )
   B=VectorParameter() {parameterName=Color Fade Friendly, defaultValue=<rgba 3 0.8897 0.3995 0>}.RGB
   Alpha=@n14.B
  )
  n16 = LinearInterpolate(
   A=LinearInterpolate(
    A=VectorParameter() {parameterName=Color Inner Enemy, defaultValue=<rgba 1 0.17 0.13 1>}.RGB
    B=VectorParameter() {parameterName=Color Outer Enemy, defaultValue=<rgba 10 2.4663 1.3167 1>}.RGB
    Alpha=@n13
   )
   B=VectorParameter() {parameterName=Color Fade Friendly, defaultValue=<rgba 4 0.0841 0.04 1>}.RGB
   Alpha=@n14.B
  )
  n17 = DynamicParameter() {defaultValue=<rgba 0 0 0 0>, paramNames [ScannerRandomPosition, Speed Random, Param3, Team Color Picker]}
  n18 = ParticleColor()
  MaterialFunctionCall(Emissive=Multiply(A=LinearInterpolate(A=@n15, B=@n16, Alpha=@n17.Team Color Picker), B=@n18.RGB)) {materialFunction=&MF_FX_EmissiveAdjustments}
  n19 = ScalarParameter() {defaultValue=0.35, parameterName=Fresnel Inner}
  n20 = ScalarParameter() {defaultValue=0.5, parameterName=Fresnel Outer}
  n21 = VertexColor()
  n22 = ScalarParameter() {defaultValue=1, parameterName=GradientFadeExponent}
  n23 = Multiply(A=@n18.A, B=Multiply(A=LinearInterpolate(A=@n19, B=@n20, Alpha=@n13), B=Power(Base=@n21.A, Exp=@n22)))
  n24 = DepthFade(Opacity=@n23, FadeDistance=ScalarParameter() {defaultValue=35, parameterName=DepthFadeDistance})
  n25 = Multiply(
   A=QualitySwitch(Default=@n24, Low=@n23, High=@n24, Medium=@n24)
   B=MaterialFunctionCall(Distance=ScalarParameter() {defaultValue=120, parameterName=NearCameraFadeDistance}) {materialFunction=&Near_Camera_Fade}.Result
  )
  n26 = StaticSwitchParameter(
   True=MaterialFunctionCall(
    Opacity In=@n25
    Fade_Start_Distance=ScalarParameter() {defaultValue=1500, parameterName=Distance Fade Start}
    Fade_End_Distance=ScalarParameter() {defaultValue=900, parameterName=Distance Fade End}
   ) {materialFunction=&MF_Distance_SyncedFade_FX}.Result Inverted
   False=@n25
  ) {parameterName=UseDistanceSyncedFade?}
  n27 = TextureSampleParameter2D(UVs=TextureCoordinate()) {
   parameterName=MaskTexture
   samplerType=SAMPLERTYPE_LinearGrayscale
   texture=/Game/Effects/Fort_Effects/Effects/Characters/Generic/ResIn/SquareGradient_Linear
  }
  n28 = Multiply(A=@n27.RGB, B=ScalarParameter() {defaultValue=5, parameterName=EdgeMaskScale, sliderMax=1}) {constB=5}
  Multiply(
   A=ScalarParameter() {defaultValue=1, parameterName=FinalOpacityMultiplier, sliderMax=1}
   B=LinearInterpolate(A=@n26, B=Multiply(A=@n26, B=@n28), Alpha=ScalarParameter() {parameterName=EdgeMaskWeight, sliderMax=1})
  )
 ]
}
