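# Unlit translucent material for the Bot Turret thruster, used on Niagara mesh particles with team-tinted glow and refraction.
# Two panned sparkle-noise samples, inverted Fresnel and vertex red form a mask blending inner/outer colors and opacity, with Dynamic Parameter TeamColor selecting friendly or enemy.
# Randomized paired sines flicker opacity and wobble vertices, weighted by vertex alpha and blue respectively.
# Mobile drops the Fresnel mask, refraction and scrolling normal.
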Material /Game/Gadgets/Assets/Bot_Turret/Materials/M_BotTurret_Thruster_01
graph {
 properties {
  bUsedWithNiagaraMeshParticles=true
  blendMode=BLEND_Translucent
  pixelDepthOffsetMode=PDOM_Legacy
  refractionMethod=RM_IndexOfRefraction
  shadingModel=MSM_Unlit
  usedShadingModels=Unlit
 }
 nodes [
  n1 = VectorParameter() {parameterName=Color Inner Friendly, defaultValue=<rgba 0.5 0.2893 0.0785 1>}
  n2 = VectorParameter() {parameterName=Color Outer Friendly, defaultValue=<rgba 10 6.47 0.7913 1>}
  n3 = TextureCoordinate() {uTiling=2, vTiling=0.1}
  n4 = TextureCoordinate() {vTiling=0.15}
  n5 = Add(
   A=TextureSample(UVs=Panner(Coordinate=@n3) {bFractionalPart=true, speedX=1.31, speedY=-2.9}) {samplerType=SAMPLERTYPE_LinearColor, texture=/Game/Gadgets/Assets/Slow_Field/Textures/T_Slowfield_SparkleNoise_01}.R
   B=TextureSample(UVs=Panner(Coordinate=@n4) {bFractionalPart=true, speedX=-0.87, speedY=-3.7}) {samplerType=SAMPLERTYPE_LinearColor, texture=&T_Slowfield_SparkleNoise_01}.G
  )
  n6 = Add(
   A=MaterialFunctionCall(
    Brightness=ScalarParameter() {defaultValue=1.04, parameterName=Fresnel Brightness}
    Exponent=ScalarParameter() {defaultValue=2, parameterName=Fresnel Exponent}
   ) {materialFunction=&MF_Fresnel_FX}.Fresnel Inverted
   B=@n5
  )
  n7 = Clamp(None=Add(A=ShadingPathSwitch(Default=@n6, Mobile=@n5) {desc=Migrated from FeatureLevelSwitch}, B=VertexColor().R))
  n8 = VectorParameter() {parameterName=Color Inner Enemy, defaultValue=<rgba 0.5 0.112 0.0785 1>}
  n9 = VectorParameter() {parameterName=Color Outer Enemy, defaultValue=<rgba 10 1.1734 0.7913 1>}
  MaterialFunctionCall(Emissive=LinearInterpolate(
    A=LinearInterpolate(A=@n1.RGB, B=@n2.RGB, Alpha=@n7)
    B=LinearInterpolate(A=@n8.RGB, B=@n9.RGB, Alpha=@n7)
    Alpha=DynamicParameter() {defaultValue=<rgba 0 0 0 1>, paramNames [Speed Random, WPO Random, Param3, TeamColor]}.TeamColor
   )) {materialFunction=&MF_FX_EmissiveAdjustments}
  n10 = ScalarParameter() {defaultValue=0.5, parameterName=Fresnel Outer}
  n11 = Multiply(
   A=Time() {period=6}
   B=DynamicParameter() {defaultValue=<rgba 0 0 0 1>, paramNames [Speed Random, WPO Random, Param3, TeamColor]}.Speed Random
  )
  n12 = Add(A=Sine(None=@n11) {period=0.1108}, B=Sine(None=@n11) {period=0.1735})
  n13 = Multiply(A=ScalarParameter() {defaultValue=0.15, parameterName=Fresnel Flicker Amount}, B=@n12)
  Multiply(
   A=LinearInterpolate(A=ScalarParameter() {defaultValue=0.1, parameterName=Fresnel Inner}, B=Add(A=@n10, B=@n13), Alpha=@n7)
   B=VertexColor().A
  )
  n14 = Multiply(
   A=VectorParameter() {parameterName=WPO Direction, defaultValue=<rgba -10 0 0 0>}.RGB
   B=DynamicParameter() {defaultValue=<rgba 0 0 0 1>, paramNames [Speed Random, WPO Random, Param3, TeamColor]}.WPO Random
  )
  Multiply(A=Multiply(A=@n12, B=Transform(None=@n14) {transformSourceType=TRANSFORMSOURCE_Local}), B=VertexColor().B)
  n15 = Constant()
  ShadingPathSwitch(
   Default=LinearInterpolate(B=ScalarParameter() {defaultValue=0.95, parameterName=Refraction Amount}, Alpha=VertexColor().A) {constA=1}
   Mobile=@n15
  ) {desc=Migrated from FeatureLevelSwitch}
  n16 = TextureCoordinate() {vTiling=0.22}
  ShadingPathSwitch(
   Default=TextureSample(UVs=Panner(Coordinate=@n16) {bFractionalPart=true, speedY=-3.8}) {samplerType=SAMPLERTYPE_Normal, texture=/Game/Global/Textures/Default/Noise/T_Noise_Normal_01}.RGB
   Mobile=@n15
  ) {desc=Migrated from FeatureLevelSwitch}
 ]
}
