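# Base surface material for the Bot Turret gadget, using Default Lit shading and the Explosive physical material.
# Constant dark grey diffuse and dark blue-tinted specular colors feed ConvertFromDiffSpec, which converts diffuse/specular inputs to PBR values.
# A separate constant and a Power node on a fixed base are also present, with no textures or parameters in the graph.
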Material /Game/Gadgets/Assets/Bot_Turret/Materials/M_Bot_Turret_Base
graph {
 properties {physMaterial=/Game/FortressPhysicalMaterials/Explosive, pixelDepthOffsetMode=PDOM_Legacy, usedShadingModels=Default Lit}
 nodes [
  Constant()
  MaterialFunctionCall(
   DiffuseColor=Constant3Vector() {constant=<rgba 0.0442 0.0442 0.0442 0>}
   SpecularColor=Constant3Vector() {constant=<rgba 0.0394 0.0547 0.1327 0>}
  ) {materialFunction=&ConvertFromDiffSpec}
  Power(Base=Constant() {r=0.5})
 ]
}
