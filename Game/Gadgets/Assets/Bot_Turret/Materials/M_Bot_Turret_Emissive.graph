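# Unlit, Nanite-enabled emissive material for the Bot Turret gadget, driven by the shared MPC_Hologram parameter collection.
# The Color parameter, amber by default, is boosted eightfold, lerped toward red by the Contested value, then multiplied by ScreensEnabled.
# A ColorSwitch reroute, a Color_A parameter and two spare lerps sit outside that chain and appear unused.
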Material /Game/Gadgets/Assets/Bot_Turret/Materials/M_Bot_Turret_Emissive
graph {
 properties {bUsedWithNanite=true, pixelDepthOffsetMode=PDOM_Legacy, shadingModel=MSM_Unlit, usedShadingModels=Unlit}
 nodes [
  LinearInterpolate()
  LinearInterpolate(Alpha=NamedRerouteUsage())
  n1 = VectorParameter() {parameterName=Color, defaultValue=<rgba 0.8584 0.457 0 1>}
  Multiply(
   A=LinearInterpolate(
    A=Multiply(A=@n1.RGB) {constB=8}
    B=Constant3Vector() {constant=<rgba 2 0 0 1>}
    Alpha=CollectionParameter() {collection=/Game/Environments/TechArt/Hologram/MPC_Hologram, parameterName=Contested}
   )
   B=CollectionParameter() {collection=&MPC_Hologram, parameterName=ScreensEnabled}
  )
  NamedRerouteDeclaration(None=CollectionParameter() {collection=&MPC_Hologram, parameterName=ColorSwitch}) {name=ColorSwitch, nodeColor=<rgba 0 0.2266 1 1>}
  VectorParameter() {parameterName=Color_A, defaultValue=<rgba 0.8584 0.457 0 1>}
 ]
}
