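# Physics asset for the Bot_Turret gadget's skeletal mesh, giving it simplified box collision bodies.
# Four boxes cover a tiny root, the main body on b_Root_Anim, and the left and right guns.
# Constraints chain the body to the root and each gun to the body, with all swing and twist motion set to free.
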PhysicsAsset /Game/Gadgets/Assets/Bot_Turret/Meshes/SK_Bot_Turret_Physics
bodies {
 b_Gun_l [{shapeType=Box, extentX=10.2908, extentY=31.6175, extentZ=18.6316, center=<0.6921 0.4026 -3.3922>}]
 b_Gun_r [{shapeType=Box, extentX=10.2908, extentY=31.6175, extentZ=18.6316, center=<0.6917 0.4025 3.5302>}]
 b_Root [{shapeType=Box, extentX=1.01, extentY=1.01, extentZ=1.01}]
 b_Root_Anim [{shapeType=Box, extentX=29.7916, extentY=41.1861, extentZ=23.9202, center=<-0.1258 -3.6568 -0.2887>}]
}
constraints [
 {
  bone1Name=b_Root_Anim
  bone2Name=b_Root
  swing1LimitDegrees=45
  swing1Motion=Free
  swing2LimitDegrees=45
  swing2Motion=Free
  twistLimitDegrees=45
  twistMotion=Free
 }
 {
  bone1Name=b_Gun_l
  bone2Name=b_Root_Anim
  swing1LimitDegrees=45
  swing1Motion=Free
  swing2LimitDegrees=45
  swing2Motion=Free
  twistLimitDegrees=45
  twistMotion=Free
 }
 {
  bone1Name=b_Gun_r
  bone2Name=b_Root_Anim
  swing1LimitDegrees=45
  swing1Motion=Free
  swing2LimitDegrees=45
  swing2Motion=Free
  twistLimitDegrees=45
  twistMotion=Free
 }
]
